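# Unlit, two-sided translucent Niagara particle material for the Galaxy Knight outfit's "bits" effect, producing starry, dissolving shapes.
# Stars come from camera-vector triplanar layers with sine-driven flicker, shaped by Fresnel and panning cell noise, then added to another rerouted layer.
# The edge mask thresholds two panning, normal-distorted noises against a warped sphere mask, with particle alpha setting smoothstep width and particle random offsetting UVs.
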Material /BRCosmetics/Characters/Player/Male/Medium/Bodies/M_MED_GalaxyKnight_Ultra/FX/M_FX_GalaxyKnight_Bits
graph {
 properties {
  bUsedWithInstancedStaticMeshes=true
  bUsedWithNiagaraMeshParticles=true
  bUsedWithNiagaraSprites=true
  blendMode=BLEND_Translucent
  opacityMaskClipValue=0.2
  pixelDepthOffsetMode=PDOM_Legacy
  shadingModel=MSM_Unlit
  twoSided=true
  usedShadingModels=Unlit
 }
 nodes [
  n1 = NamedRerouteUsage()
  n2 = ScalarParameter() {parameterName=Lower_End}
  n3 = Subtract(A=ScalarParameter() {defaultValue=0.2, parameterName=Upper_End}, B=@n2)
  Clamp(None=Divide(A=Subtract(A=@n1, B=@n2), B=Clamp(None=@n3)))
  MaterialFunctionCall(Emissive=NamedRerouteUsage()) {materialFunction=&MF_FX_EmissiveAdjustments}
  n4 = Constant()
  n5 = TextureObject() {texture=/Game/Characters/Player/Male/Medium/Bodies/M_MED_HighTower_Tapas/Textures/T_Hightower_Tapas_Star}
  n6 = ScalarParameter() {defaultValue=1, group=Stars, parameterName=SmallStarsTiling}
  n7 = Constant()
  n8 = Constant3Vector() {constant=<rgba 0.5 0.5 0.5 0>}
  n9 = Time()
  n10 = ScalarParameter() {defaultValue=0.3, group=Stars, parameterName=StarFlickeringSpeed}
  n11 = Sine(None=Frac(None=Multiply(A=@n9, B=@n10) {constB=0.3}))
  n12 = Add(
   A=LinearInterpolate(
    A=MaterialFunctionCall!(CameraVector, Rotation Speed=@n4, Texture=@n5, Tiling=@n6, Offset, AxisFadeContrast) {materialFunction=/Game/Weapons/WeaponSkins/MaterialFunction/TriplanarCameraVector_Rotated}.Result
    B=MaterialFunctionCall(Rotation Speed=@n7, Texture=@n5, Tiling=@n6, Offset=@n8) {materialFunction=&TriplanarCameraVector_Rotated}.Result
    Alpha=ConstantBiasScale(None=@n11)
   )
   B=MaterialFunctionCall(Rotation Speed=Constant(), Texture=@n5, Tiling=ScalarParameter() {defaultValue=0.75, group=Stars, parameterName=BigStarsTiling}) {materialFunction=&TriplanarCameraVector_Rotated}.Result
  )
  n13 = Constant() {r=1}
  n14 = Constant()
  n15 = Fresnel(ExponentIn=ScalarParameter() {defaultValue=0.5, group=Stars, parameterName=StarFresnelExp}, Normal=VertexNormalWS()) {baseReflectFraction=0.05, exponent=1}
  n16 = OneMinus(None=Saturate(None=LinearInterpolate(A=@n13, B=@n14, Alpha=@n15) {constA=10, constB=0}))
  n17 = VectorParameter() {group=Stars, parameterName=UV Scale, defaultValue=<rgba 0.125 0.125 0 1>}
  n18 = Panner(Coordinate=Multiply(
    A=TextureCoordinate()
    B=MaterialFunctionCall(X=@n17.R, Y=@n17.R) {materialFunction=&MakeFloat2}.Result
   )) {speedX=0.1, speedY=0.1}
  n19 = Multiply(
   A=Multiply(
    A=Multiply(A=@n12, B=@n16)
    B=TextureSample(UVs=@n18) {samplerType=SAMPLERTYPE_Grayscale, texture=/BRCosmetics/Weapons/FORT_Melee/Pickaxe_Cube_Ninja_Male/FX/T_CellNoise}.R
   )
   B=ScalarParameter() {defaultValue=15, group=Stars, parameterName=Emission}
  )
  NamedRerouteDeclaration(None=Add(A=NamedRerouteDeclaration(None=@n19) {name=Star, nodeColor=<rgba 0 1 0.6484 1>}, B=NamedRerouteUsage())) {name=Galaxy, nodeColor=<rgba 0.1016 1 0 1>}
  n20 = ScalarParameter() {defaultValue=0.2, group=Opacity Mask, parameterName=Edge Level}
  n21 = ScalarParameter() {defaultValue=1, group=Opacity Mask, parameterName=Edge Sofness}
  n22 = Add(A=TextureCoordinate() {uTiling=2}, B=ParticleRandom())
  n23 = VectorParameter() {group=Opacity Mask, parameterName=UV Scale, defaultValue=<rgba 0.125 0.125 0 1>}
  n24 = MaterialFunctionCall(X=@n23.R, Y=@n23.R) {materialFunction=&MakeFloat2}
  n25 = TextureSample(UVs=Panner(Coordinate=Multiply(A=@n22, B=@n24.Result)) {speedX=-0.1, speedY=0.1}) {samplerType=SAMPLERTYPE_Normal, texture=/Game/Global/Textures/Default/Tileable_Patterns/1Normal}
  n26 = MaterialFunctionCall(X=@n25.R, Y=@n25.R) {materialFunction=&MakeFloat2}
  n27 = SphereMask(A=Add(A=TextureCoordinate(), B=Multiply(A=@n26.Result) {constB=0.25}), B=Constant2Vector() {g=0.5, r=0.5}) {attenuationRadius=0.45, hardnessPercent=0}
  n28 = Saturate(None=SmoothStep(Min=@n20, Max=Add(A=@n20, B=@n21), Value=OneMinus(None=@n27)))
  n29 = Add(A=@n28, B=OneMinus(None=ParticleColor().A))
  n30 = Add(A=TextureCoordinate(), B=ParticleRandom())
  n31 = VectorParameter() {group=Opacity Mask, parameterName=UV2 Scale, defaultValue=<rgba 0.5 0.5 0 1>}
  n32 = MaterialFunctionCall(X=@n31.R, Y=@n31.R) {materialFunction=&MakeFloat2}
  n33 = Add(A=Add(A=Multiply(A=@n30, B=@n32.Result)), B=Multiply(A=@n26.Result) {constB=0.1}) {constB=0}
  n34 = TextureSample(UVs=Panner(Coordinate=@n33) {speedX=0.1, speedY=0.3}) {
   samplerType=SAMPLERTYPE_LinearGrayscale
   texture=/Game/Characters/Player/Male/Medium/Bodies/M_MED_Citadel/FX/T_Citadel_Noise_Dissolve_Handpaint
  }
  n35 = Panner(Coordinate=@n33) {speedX=0.05, speedY=0.25}
  n36 = Multiply(A=OneMinus(None=TextureSample(UVs=@n35) {texture=/Game/TimeOfDay/LightFunction/T_CloudNoise_Artemis}.G))
  n37 = Saturate(None=Power(Base=Multiply(A=@n34.R, B=@n36), Exp=ScalarParameter() {defaultValue=1.2, group=Opacity Mask, parameterName=Texture Exponent}))
  NamedRerouteDeclaration(None=Multiply(A=SmoothStep(Min=@n28, Max=@n29, Value=@n37))) {name=Edge_Mask, nodeColor=<rgba 1 0 0.3516 1>}
  NamedRerouteDeclaration(None=ParticleColor().RGB) {name=BaseColor, nodeColor=<rgba 1 0.1875 0 1>}
 ]
}
