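# Translucent unlit material for the Galaxy Knight's flaming head, usable on Niagara mesh particles.
# Emissive layers build a two-color fresnel gradient, scrolling distorted galaxy smoke, and twinkling camera-vector triplanar stars.
# Panning noise, UV gradients and erosion carve the flame's opacity, with OpacityMult also driving dissolve.
# World position offset combines two panning-noise displacements with a pendulum bend fed by dynamic parameters, using a previous-frame switch for correct velocity.
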Material /BRCosmetics/Characters/Player/Male/Medium/Bodies/M_MED_GalaxyKnight_Ultra/FX/M_FX_GalaxyKnight_FlamingHead
graph {
 properties {
  allowTranslucentCustomDepthWrites=true
  bUsedWithInstancedStaticMeshes=true
  bUsedWithNiagaraMeshParticles=true
  blendMode=BLEND_Translucent
  opacityMaskClipValue=0.2
  pixelDepthOffsetMode=PDOM_Legacy
  shadingModel=MSM_Unlit
  usedShadingModels=Unlit
 }
 nodes [
  Add(A=Add(A=NamedRerouteUsage(), B=NamedRerouteUsage()), B=NamedRerouteUsage())
  MaterialFunctionCall(Emissive=NamedRerouteUsage()) {materialFunction=&MF_FX_EmissiveAdjustments}
  n1 = ScalarParameter() {parameterName=Lower_End}
  n2 = Subtract(A=NamedRerouteUsage(), B=@n1)
  n3 = Clamp(None=Subtract(A=ScalarParameter() {defaultValue=0.2, parameterName=Upper_End}, B=@n1))
  Multiply(A=Clamp(None=Divide(A=@n2, B=@n3)), B=ScalarParameter() {defaultValue=1, parameterName=OpacityMult})
  n4 = TextureObject() {texture=/Game/Characters/Player/Male/Medium/Bodies/M_MED_HighTower_Tapas/Textures/T_Hightower_Tapas_Star}
  n5 = MaterialFunctionCall!(CameraVector, Rotation Speed=Constant(), Texture=@n4, Tiling=Constant() {r=1}, Offset, AxisFadeContrast) {materialFunction=/Game/Weapons/WeaponSkins/MaterialFunction/TriplanarCameraVector_Rotated}
  n6 = MaterialFunctionCall(Rotation Speed=Constant(), Texture=@n4, Tiling=Constant() {r=2}, Offset=Constant3Vector() {constant=<rgba 0.5 0.5 0.5 0>}) {materialFunction=&TriplanarCameraVector_Rotated}
  n7 = Multiply(A=Time(), B=Constant() {r=0.3}) {constB=0.3}
  n8 = ConstantBiasScale(None=Sine(None=Frac(None=@n7)))
  n9 = Constant()
  n10 = Constant() {r=0.9}
  n11 = LinearInterpolate(
   A=Constant() {r=1}
   B=Constant()
   Alpha=Fresnel(ExponentIn=Constant() {r=0.5}, Normal=VertexNormalWS()) {baseReflectFraction=0.05, exponent=1}
  ) {constA=10, constB=0}
  n12 = Multiply(
   A=Add(
    A=LinearInterpolate(A=@n5.Result, B=@n6.Result, Alpha=@n8)
    B=MaterialFunctionCall(Rotation Speed=@n9, Texture=@n4, Tiling=@n10) {materialFunction=&TriplanarCameraVector_Rotated}.Result
   )
   B=OneMinus(None=Saturate(None=@n11))
  )
  n13 = TextureCoordinate()
  n14 = Constant3Vector() {constant=<rgba 0.5 0.25 0 0>}
  n15 = MaterialFunctionCall(X=@n14.R, Y=@n14.R) {materialFunction=&MakeFloat2}
  n16 = TextureSample(UVs=Panner(Coordinate=Multiply(A=@n13, B=@n15.Result)) {speedX=0.1, speedY=0.1}) {samplerType=SAMPLERTYPE_Grayscale, texture=/BRCosmetics/Weapons/FORT_Melee/Pickaxe_Cube_Ninja_Male/FX/T_CellNoise}
  NamedRerouteDeclaration(None=Multiply(A=Multiply(A=@n12, B=@n16.R), B=ScalarParameter() {defaultValue=15, group=GalaxyKnight, parameterName=StarEmissive})) {name=Star, nodeColor=<rgba 0 1 0.6484 1>}
  n17 = NamedRerouteUsage()
  n18 = TextureCoordinate() {uTiling=2}
  n19 = VectorParameter() {group=Galaxy, parameterName=DistortionUVScale, defaultValue=<rgba 0.25 0.2 0 1>}
  n20 = MaterialFunctionCall(X=@n19.R, Y=@n19.R) {materialFunction=&MakeFloat2}
  n21 = TextureSample(UVs=Panner(Coordinate=Multiply(A=@n18, B=@n20.Result)) {speedY=0.025}) {samplerType=SAMPLERTYPE_Normal, texture=/Game/Global/Textures/Default/Tileable_Patterns/1Normal}
  n22 = MaterialFunctionCall(X=@n21.R, Y=@n21.R) {materialFunction=&MakeFloat2}
  n23 = ScalarParameter() {defaultValue=0.1, group=Galaxy, parameterName=GalaxyDistortion}
  n24 = VectorParameter() {group=GalaxyKnight, parameterName=GalaxyUVScale, defaultValue=<rgba 4.5 2 0.536 0>}
  n25 = Multiply(
   A=Add(A=TextureCoordinate(), B=Multiply(A=@n22.Result, B=@n23) {constB=0.025})
   B=MaterialFunctionCall(X=@n24.R, Y=@n24.R) {materialFunction=&MakeFloat2}.Result
  )
  n26 = MaterialFunctionCall(X=@n24.B, Y=@n24.B) {materialFunction=&MakeFloat2}
  n27 = VectorParameter() {group=Galaxy, parameterName=GalaxySpeed, defaultValue=<rgba 0.2 0 0 0>}
  n28 = TextureSample(
   UVs=Panner(
    Coordinate=Add(A=@n25, B=@n26.Result)
    Speed=MaterialFunctionCall(X=@n27.R, Y=@n27.R) {materialFunction=&MakeFloat2}.Result
   )
   Tex=TextureObjectParameter() {group=Galaxy, parameterName=GalaxyTexture, texture=&T_Inky_Smoke_Tile}
  )
  MaterialExpressionRemap_0 = MaterialXRemap!(None=ScalarParameter() {defaultValue=1, parameterName=EmissiveMult}, InputLow, InputHigh, TargetLow, TargetHigh) {inputHighDefault=3, inputLowDefault=1, targetHighDefault=4, targetLowDefault=1}
  n29 = Multiply(A=Multiply(A=Desaturation(None=@n28.RGB), B=@MaterialExpressionRemap_0), B=NamedRerouteUsage())
  n30 = Add(A=@n17, B=Multiply(A=Multiply(A=@n29, B=@n17), B=ScalarParameter() {defaultValue=2, group=Galaxy, parameterName=Galaxy mult}) {constB=5})
  n31 = Multiply(A=Multiply(A=@n30, B=NamedRerouteUsage()), B=ScalarParameter() {defaultValue=20, group=Galaxy, parameterName=StarMult}) {constB=100}
  NamedRerouteDeclaration(None=Multiply(A=Add(A=@n30, B=@n31), B=ScalarParameter() {defaultValue=1, parameterName=EmissiveMult})) {name=GalaxyEmissive, nodeColor=<rgba 0 1 0.3438 1>}
  n32 = ScalarParameter() {defaultValue=0.12, group=WPO Mask, parameterName=MaskB Amount}
  n33 = ScalarParameter() {defaultValue=0.2, group=WPO Mask, parameterName=MaskB Softness}
  n34 = ComponentMask(None=TextureCoordinate()) {r=true}
  NamedRerouteDeclaration(None=SmoothStep(Min=@n32, Max=Add(A=@n32, B=@n33), Value=OneMinus(None=@n34))) {name=MaskEnds, nodeColor=<rgba 0.7812 1 0 1>}
  n35 = ScalarParameter() {defaultValue=0.2, group=Opacity Mask, parameterName=Edge Level}
  n36 = ComponentMask(None=TextureCoordinate()) {r=true}
  n37 = SmoothStep(Min=@n35, Max=Add(A=@n35, B=ScalarParameter() {defaultValue=1.8, group=Opacity Mask, parameterName=Edge Sofness}), Value=OneMinus(None=@n36))
  n38 = ScalarParameter() {defaultValue=0.5, group=Opacity Mask, parameterName=Bottom Mask Level}
  n39 = SmoothStep(Min=@n38, Max=Add(A=@n38, B=ScalarParameter() {defaultValue=1, group=Opacity Mask, parameterName=Bottom Sofness}), Value=@n36)
  n40 = Saturate(None=StaticSwitchParameter(True=Max(A=@n37, B=@n39), False=@n37) {group=Bottom Mask, parameterName=Use bottom mask?})
  n41 = TextureCoordinate()
  n42 = VectorParameter() {group=Opacity Mask, parameterName=UV2 Scale, defaultValue=<rgba 1 1 0 1>}
  n43 = MaterialFunctionCall(X=@n42.R, Y=@n42.R) {materialFunction=&MakeFloat2}
  n44 = TextureCoordinate() {uTiling=2}
  n45 = VectorParameter() {group=Opacity Mask, parameterName=UV Scale, defaultValue=<rgba 0.25 0.2 0 1>}
  n46 = MaterialFunctionCall(X=@n45.R, Y=@n45.R) {materialFunction=&MakeFloat2}
  n47 = TextureSample(UVs=Panner(Coordinate=Multiply(A=@n44, B=@n46.Result)) {speedX=-0.1, speedY=0.1}) {samplerType=SAMPLERTYPE_Normal, texture=&1Normal}
  n48 = Multiply(A=MaterialFunctionCall(X=@n47.R, Y=@n47.R) {materialFunction=&MakeFloat2}.Result) {constB=0.1}
  n49 = Add(A=Add(A=Multiply(A=@n41, B=@n43.Result)), B=@n48) {constB=0}
  n50 = TextureSample(UVs=Panner(Coordinate=@n49) {speedX=0.5, speedY=0.1}) {
   samplerType=SAMPLERTYPE_LinearGrayscale
   texture=/Game/Characters/Player/Male/Medium/Bodies/M_MED_Citadel/FX/T_Citadel_Noise_Dissolve_Handpaint
  }
  n51 = Panner(Coordinate=@n49) {speedX=0.5, speedY=0.05}
  n52 = Multiply(A=OneMinus(None=TextureSample(UVs=@n51) {texture=/Game/TimeOfDay/LightFunction/T_CloudNoise_Artemis}.G))
  n53 = Saturate(None=Power(Base=Multiply(A=@n50.R, B=@n52), Exp=Constant() {r=1.2}))
  n54 = SmoothStep(Min=@n40, Max=Add(A=@n40, B=ScalarParameter() {defaultValue=0.1, group=Opacity Mask, parameterName=Soft}), Value=@n53)
  n55 = TextureCoordinate()
  n56 = VectorParameter() {group=Opacity Mask, parameterName=UV2 Scale, defaultValue=<rgba 1 1 0 1>}
  n57 = MaterialFunctionCall(X=@n56.R, Y=@n56.R) {materialFunction=&MakeFloat2}
  n58 = Add(A=Add(A=Multiply(A=@n55, B=@n57.Result), B=@n48)) {constB=0}
  n59 = TextureSample(UVs=Panner(Coordinate=@n58) {speedX=0.5, speedY=0.1}) {
   samplerType=SAMPLERTYPE_LinearGrayscale
   texture=&T_Citadel_Noise_Dissolve_Handpaint
  }
  n60 = TextureSample(UVs=Panner(Coordinate=@n58) {speedX=0.5, speedY=0.05}) {texture=&T_CloudNoise_Artemis}
  n61 = Multiply(A=@n59.R, B=Multiply(A=OneMinus(None=@n60.G)) {constB=2})
  n62 = Constant() {r=1.2}
  n63 = TextureCoordinate()
  n64 = Multiply(A=Distance(A=ComponentMask(None=@n63) {g=true}, B=Constant() {r=0.5})) {constB=2}
  n65 = ComponentMask(None=TextureCoordinate()) {r=true}
  n66 = LinearInterpolate(A=Constant() {r=0.1}, B=Constant() {r=4}, Alpha=Saturate(None=Multiply(A=@n65) {constB=5})) {constA=20}
  n67 = Saturate(None=MaterialFunctionCall!(
    Gradient=Power(Base=@n61, Exp=@n62)
    Erosion Amount=Power(Base=@n64, Exp=@n66)
    Softness=ScalarParameter() {defaultValue=1, group=Opacity Mask, parameterName=Erosion Softness}
   ) {materialFunction=/BRCosmetics/Athena/Effects/Particles/Skydiving/NS_Contrail_DarkAndLight/Functions/MF_Erosion}.Increased Range)
  n68 = Constant() {r=1}
  MaterialExpressionRemap_3 = MaterialXRemap(None=OneMinus(None=ScalarParameter() {defaultValue=1, parameterName=OpacityMult})) {targetLowDefault=-0.5}
  n69 = TextureCoordinate()
  MaterialExpressionRemap_1 = MaterialXRemap(None=SmoothStep(
    Min=@MaterialExpressionRemap_3
    Max=Add(A=@MaterialExpressionRemap_3, B=@n68)
    Value=MaterialFunctionCall(Float2=@n69) {materialFunction=&BreakOutFloat2Components}.R
   )) {inputHighDefault=0, inputLowDefault=1, targetHighDefault=0.5}
  n70 = SmoothStep(Min=@MaterialExpressionRemap_1, Max=Add(A=@MaterialExpressionRemap_1, B=Constant() {r=0.2}), Value=@n53)
  NamedRerouteDeclaration(None=Multiply(A=Multiply(A=@n54, B=@n67), B=Saturate(None=@n70))) {name=Edge_Mask, nodeColor=<rgba 1 0 0.3516 1>}
  n71 = DynamicParameter() {parameterIndex=1, defaultValue=<rgba 0 0 -1 1>, paramNames [RestNormalX, RestNormalY, RestNormalZ, --------------]}
  n72 = Multiply(A=MaterialFunctionCall(R=@n71.RestNormalX, G=@n71.RestNormalX, B=@n71.RestNormalX) {materialFunction=&AppendMany}.RGB) {constB=-1}
  n73 = DynamicParameter() {parameterIndex=3, defaultValue=<rgba 1 1 1 0.15>, paramNames [PosX,PosY,PosZ,Rigidity]}
  n74 = DynamicParameter() {parameterIndex=2, paramNames [MeshOriginX,MeshOriginY,MeshOriginZ,MaskFalloff]}
  n75 = MaterialFunctionCall(R=@n74.MeshOriginX, G=@n74.MeshOriginX, B=@n74.MeshOriginX, A=@n74.MeshOriginX) {materialFunction=&AppendMany}
  n76 = Subtract(
   A=MaterialFunctionCall(R=@n73.PosX, G=@n73.PosX, B=@n73.PosX) {materialFunction=&AppendMany}.RGB
   B=@n75.RGB
  )
  n77 = WorldPosition()
  n78 = RotateAboutAxis(
   NormalizedRotationAxis=CrossProduct(A=@n72, B=Normalize(VectorInput=@n76))
   RotationAngle=@n73.Rigidity
   PivotPoint=@n75.RGB
   Position=@n77.XYZ
  )
  n79 = ComponentMask(None=TextureCoordinate()) {r=true}
  n80 = Saturate(None=Subtract(A=OneMinus(None=@n79), B=ScalarParameter() {group=Pendulum, parameterName=Mask Offset}))
  n81 = ScalarParameter() {defaultValue=1, group=Pendulum, parameterName=Power}
  n82 = Clamp(None=Saturate(None=Power(Base=@n80, Exp=@n81)))
  n83 = DynamicParameter() {paramNames [PrevPosX,PrevPosY,PrevPosZ,GlowAmt]}
  n84 = Subtract(
   A=MaterialFunctionCall(R=@n83.PrevPosX, G=@n83.PrevPosX, B=@n83.PrevPosX) {materialFunction=&AppendMany}.RGB
   B=@n75.RGB
  )
  n85 = RotateAboutAxis(
   NormalizedRotationAxis=CrossProduct(A=@n72, B=Normalize(VectorInput=@n84))
   RotationAngle=@n73.Rigidity
   PivotPoint=@n75.RGB
   Position=@n77.XYZ
  )
  NamedRerouteDeclaration(None=PreviousFrameSwitch(Current Frame=LinearInterpolate(A=@n78, Alpha=@n82) {constB=0}, Previous Frame=LinearInterpolate(A=@n85, Alpha=@n82) {constB=0})) {name=Pendulum, nodeColor=<rgba 0 1 0.5313 1>}
  n86 = Multiply(
   A=TextureCoordinate() {uTiling=0.5}
   B=ComponentMask(None=VectorParameter() {group=WPO1, parameterName=WPO1_UVScale, defaultValue=<rgba 1 0 0 0>}.RGB) {g=true, r=true}
  )
  n87 = ComponentMask(None=VectorParameter() {group=WPO1, parameterName=WPO1_Speed, defaultValue=<rgba 0.3 0 0 0>}.RGB) {g=true, r=true}
  n88 = TextureObject() {samplerType=SAMPLERTYPE_Normal, texture=&T_Weap_Distort_N_02}
  n89 = LinearInterpolate(
   A=ScalarParameter() {defaultValue=0.1, group=WPO1, parameterName=WPO1_Low}
   B=ScalarParameter() {defaultValue=0.5, group=WPO1, parameterName=WPO1_High}
   Alpha=TextureSample(UVs=Panner(Coordinate=@n86, Speed=@n87) {speedY=0.5}, Tex=@n88) {samplerType=SAMPLERTYPE_Normal, texture=&T_Weap_Distort_N_02}.R
  ) {constA=0.5, constB=1.5}
  n90 = NamedRerouteUsage()
  n91 = TextureCoordinate()
  n92 = Multiply(
   A=Multiply(A=Multiply(A=@n89, B=@n90), B=VectorParameter() {group=WPO1, parameterName=WPO1_direction, defaultValue=<rgba 0 0 1 1>}.RGB)
   B=LinearInterpolate(Alpha=MaterialFunctionCall(Float2=@n91) {materialFunction=&BreakOutFloat2Components}.R) {constA=0.15}
  )
  n93 = ScalarParameter() {defaultValue=30, group=WPO1, parameterName=WPO1_Strength}
  NamedRerouteDeclaration(None=Transform(None=Multiply(A=@n92, B=@n93) {constB=40})) {name=WPO_1, nodeColor=<rgba 0 1 0.0391 1>}
  n94 = ScalarParameter() {defaultValue=0.1, group=WPO2, parameterName=WPO2_Low}
  n95 = ScalarParameter() {defaultValue=0.5, group=WPO2, parameterName=WPO2_High}
  n96 = Add(A=TextureCoordinate() {uTiling=0.5}) {constB=0.5}
  n97 = ComponentMask(None=VectorParameter() {group=WPO2, parameterName=WPO2_UVScale, defaultValue=<rgba 1 0 0 0>}.RGB) {g=true, r=true}
  n98 = VectorParameter() {group=WPO2, parameterName=WPO2_Speed, defaultValue=<rgba 0.3 0 0 0>}
  n99 = TextureSample(UVs=Panner(Coordinate=Multiply(A=@n96, B=@n97), Speed=ComponentMask(None=@n98.RGB) {g=true, r=true}) {speedY=0.5}, Tex=@n88) {samplerType=SAMPLERTYPE_Normal, texture=&T_Weap_Distort_N_02}
  n100 = Multiply(
   A=Multiply(A=LinearInterpolate(A=@n94, B=@n95, Alpha=@n99.R) {constA=0.5, constB=1.5}, B=NamedRerouteUsage())
   B=VectorParameter() {group=WPO2, parameterName=WPO2_direction, defaultValue=<rgba 0 0 1 1>}.RGB
  )
  n101 = Constant() {r=0.55}
  n102 = Add(A=@n101, B=Constant() {r=0.4})
  n103 = MaterialFunctionCall(Float2=TextureCoordinate()) {materialFunction=&BreakOutFloat2Components}
  n104 = LinearInterpolate(Alpha=Saturate(None=SmoothStep(Min=@n101, Max=@n102, Value=@n103.R))) {constA=0.4}
  NamedRerouteDeclaration(None=Multiply(A=Multiply(A=@n100, B=@n104), B=ScalarParameter() {defaultValue=45, group=WPO2, parameterName=WPO2_Strength}) {constB=20}) {name=WPO_2, nodeColor=<rgba 0 1 0.0391 1>}
  n105 = ScalarParameter() {defaultValue=0.5, group=Base Color, parameterName=Color Edge Level}
  n106 = ScalarParameter() {defaultValue=0.6, group=Base Color, parameterName=Color Edge Softness}
  n107 = TextureCoordinate()
  n108 = NamedRerouteDeclaration(None=SmoothStep(
    Min=@n105
    Max=Add(A=@n105, B=@n106)
    Value=MaterialFunctionCall(Float2=@n107) {materialFunction=&BreakOutFloat2Components}.R
   )) {name=HairMask, nodeColor=<rgba 0 1 0.0859 1>}
  n109 = PixelNormalWS()
  n110 = Multiply(A=@n109) {constB=-1}
  n111 = OneMinus(None=Fresnel(
    ExponentIn=ScalarParameter() {defaultValue=0.4, group=Base Color, parameterName=Fresnel Power}
    Normal=StaticSwitchParameter(True=@n110, False=@n109) {parameterName=Reverse Normals}
   ) {baseReflectFraction=0, exponent=1})
  NamedRerouteDeclaration(None=LinearInterpolate(
    A=VectorParameter() {group=Base Color, parameterName=Color 2, defaultValue=<rgba 0.0473 0.0026 0.1667 1>}.RGB
    B=VectorParameter() {group=Base Color, parameterName=Color 1, defaultValue=<rgba 0.577 0.1398 0.671 1>}.RGB
    Alpha=Multiply(A=@n108, B=@n111)
   )) {name=BaseColor, nodeColor=<rgba 1 0.1875 0 1>}
 ]
}
