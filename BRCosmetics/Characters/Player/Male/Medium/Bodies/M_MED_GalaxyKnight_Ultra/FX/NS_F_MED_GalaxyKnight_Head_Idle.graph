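# Idle head FX for the Galaxy Knight Ultra character: simulated hair, drifting purple sprite bits and a purple light.
# Hair is one persistent particle following the skeletal mesh, driven by spring pendulum and cone constraints that orient its meshes.
# Dynamic material parameters pass pivot, axis and current/previous positions to the hair material.
# User.IsSkydiving tightens the cones and fixes hair facing; User.bisHeadHidden drives visibility tags and disables the light.
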NiagaraSystem /BRCosmetics/Characters/Player/Male/Medium/Bodies/M_MED_GalaxyKnight_Ultra/FX/NS_F_MED_GalaxyKnight_Head_Idle
dependencies {
 usedDataInterfaces [&NiagaraDataInterfaceCurve]
 usedDynamicInputs []
 usedModules [
  &SystemState
  &self:SystemSpawnScript.SystemScriptSource.SystemScriptGraph.NiagaraNodeAssignment_0.SetVariables_#1
 ]
}
emitters {
 Bits {
  emitterUpdateScript [
   &EmitterState(Life Cycle Mode=::NewEnumerator1, Loop Duration=2, UseLoopDelay=false)
   &SpawnRate(SpawnRate=2, Use Spawn Probability=false)
  ]
  particleSpawnScript [
   &InitializeParticle(
    Color=&RandomRangeLinearColor(Minimum=<rgba 0.1239 0.0346 0.258 1>, Maximum=<rgba 0.0513 0.0055 0.12 1>)
    Lifetime Max=3
    Lifetime Mode=::NewEnumerator1
    Sprite Rotation Mode=::NewEnumerator2
    Sprite Size=&Multiply_Vector2D_ByFloat(
     Float=&RandomRangeFloat(Minimum=1, Maximum=0.5)
     Vector2D=&RandomRangeVector2D(Minimum=<4 14>, Maximum=<3 7>)
    )
    Sprite Size Mode=::NewEnumerator4
    UsePositionOffset=false
    Write Material Random=true
   )
   &ShapeLocation(
    Cylinder Height=10
    Cylinder Height Midpoint=0
    Cylinder Radius=9
    Hemicircle Internal Cap=false
    Hemicircle Y=true
    MaskSpawnGroup=false
    Non Uniform Scale=<0.7 1 1>
    Offset=<-12 0 8>
    Offset Coordinate Space=::Local
    Offset Mode=::NewEnumerator0
    Override Seed=false
    Rotation Mode=::NewEnumerator2
    Shape Primitive=::NewEnumerator1
    Surface Only=true
    Use Endcaps In Surface Only Mode=false
    UseOwnerScale=true
    Yaw / Pitch / Roll=<0.25 0 -0.3>
   ) {module=ShapeLocation001}
   &AddVelocity(
    Module.Use Falloff Distance=false
    Origin Offset=&RandomRangeVector(Minimum=<40 0 8>, Maximum=<40 0 25>)
    Override Seed=false
    Velocity Mode=::NewEnumerator1
    Velocity Speed=&RandomRangeFloat(Minimum=40, Maximum=20)
   ) {module=AddVelocity001}
   &ApplyInitialForces()
  ]
  particleUpdateScript [
   &ParticleState()
   &ScaleSpriteSize(Uniform Curve Sprite Scale=DataInterface() {propertyValues {Curve {keys [<at 0 0>, {tangentMode=RCTM_User, time=0.1, value=1}, {tangentMode=RCTM_User, time=1, value=0.7}]}}})
   &Drag(Drag=&RandomRangeFloat(Minimum=0.25, Maximum=0.8), Use Linear Drag=true, Use Rotational Drag=false)
   &ScaleColor(
    Scale Alpha=&FloatFromCurve(
     FloatCurve=DataInterface() {propertyValues {Curve {keys [<at 0.25 1 RCIM_Cubic>,<at 1 0 RCIM_Cubic>]}}}
     Scale Curve=&Multiply_Float(A=1, B=$User.OpacityMult)
    )
    Scale RGB=&Multiply_VectorByFloat(Vector=<1 1 1>, Float=$User.EmissiveMult)
    ScaleA=true
    ScaleRGB=true
   )
   &SolveForcesAndVelocity(Clamp Velocity=false, Limit Acceleration=false)
   /SetVariables(Particles.VisibilityTag=&MakeCustomIntFromBool(Bool=$User.bisHeadHidden, True INT=1, False INT=0)) {isSetParameters=true, module=SetVariables_#2}
  ]
  rendererClasses [&NiagaraSpriteRendererProperties]
  renderers [{rendererClass=&NiagaraSpriteRendererProperties}]
 }
 Hair {
  emitterUpdateScript [
   &EmitterState(Life Cycle Mode=::NewEnumerator1, Loop Behavior=::NewEnumerator1, Loop Duration Mode=::NewEnumerator1, UseLoopDelay=false)
   &SpawnBurst_Instantaneous(Spawn Count=1, Use Spawn Probability=false)
  ]
  particleSpawnScript [
   /SetVariables(Particles.PendulumDrag=1, Particles.PendulumLengthScale=1, Particles.PendulumMaskFalloff=3, Particles.PendulumRigidity=0.125) {isSetParameters=true}
   /Niagara/Modules/Spawn/Initialization/InitializeParticle(
    Lifetime=999
    Mesh Scale=&Multiply_Vector(A=<1 1 1>, B=&MakeVector(X=1, Y=1, Z=$Particles.PendulumLengthScale))
    Module.Write Color=true
    Module.Write Lifetime=true
    Module.Write Mass=true
    Module.Write Material Random=true
    Module.Write Position=false
    Module.Write Scale=true
    Module.Write SpriteRotation=false
    Module.Write SpriteSize=true
    Sprite Size=<5 5>
   )
  ]
  particleUpdateScript [
   /SetVariables(Particles.HairDir:Vector3f=&TransformVector(Vector=<-1 0 0>, Destination Space=::World)) {isSetParameters=true}
   &ParticleState(Kill Particles When Lifetime Has Elapsed=false)
   &SkeletalMeshLocation(
    MaskSpawnGroup=false
    Mesh Sampling Type=::NewEnumerator5
    Orientation Sampling=::NewEnumerator1
    Position Sampling=::NewEnumerator2
    Skeletal Mesh=$User.SkeletalMesh
    Socket Sampling Mode=::NewEnumerator5
   )
   /SetVariables(Particles.MeshOrigin:Vector3f=&AddVectorToPosition(Position=$Output.SkeletalMeshLocation.SampledPositionWithoutOffset, Vector=<0 0 0>)) {isSetParameters=true}
   &GravityForce(Gravity=<0 0 -2000>)
   &Drag(Drag=$Particles.PendulumDrag, Use Linear Drag=true, Use Rotational Drag=false)
   &PendulumSetup(
    Module.Enable Spring Driver=true
    Pendulum Length:NiagaraFloat=&Multiply_Float(A=100, B=$Particles.PendulumLengthScale)
    Pendulum Pivot:Vector3f=$Particles.MeshOrigin
    Pendulum Rest Axis:Vector3f=$Particles.HairDir
    "Spring Driven Constraint (Tightness)"=6
   )
   &InheritSourceMovement(
    Applied Position Falloff Distance=10
    Applied Position Scale=&Multiply_VectorByFloat(
     Vector=<-0.5 -0.5 -0.5>
     Float=&Lerp_Float(A=1, B=0, Alpha=&MakeCustomFloatFromBool(Bool=$User.IsSkydiving, True Float=1, False Float=0))
    )
    ApplyForce=false
    ApplyPosition=true
    ApplyPositionFalloff=true
    ApplyPositionFalloffCurve=false
   ) {module=InheritSourceMovement001}
   &CurlNoiseForce(Noise Strength=100, Pan Noise=false)
   &SolveForcesAndVelocity(Acceleration Limit=10, Clamp Velocity=false, Limit Acceleration=false)
   &PendulumConstraint()
   &self:ConeConstraint_01(
    Cone Axis:Vector3f=$Particles.HairDir
    Max Angle:NiagaraFloat=&Lerp_Float(A=40, B=20, Alpha=&MakeCustomFloatFromBool(Bool=$User.IsSkydiving, True Float=1, False Float=0))
    Max Orient Angle:NiagaraFloat=&Lerp_Float(A=60, B=10, Alpha=&MakeCustomFloatFromBool(Bool=$User.IsSkydiving, True Float=1, False Float=0))
    Pivot:Vector3f=$Particles.MeshOrigin
   ) {module=ConeConstraint_01}
   &self:ConeConstraint_03(
    Cone Axis:Vector3f=$Particles.HairDir
    Max Angle:NiagaraFloat=&Lerp_Float(A=20, B=12, Alpha=&MakeCustomFloatFromBool(Bool=$User.IsSkydiving, True Float=1, False Float=0))
    Max Orient Angle:NiagaraFloat=&Lerp_Float(A=30, B=10, Alpha=&MakeCustomFloatFromBool(Bool=$User.IsSkydiving, True Float=1, False Float=0))
    Pivot:Vector3f=$Particles.MeshOrigin
   ) {module=ConeConstraint_03}
   &CalculateAccurateVelocity()
   /SetVariables(Particles.MeshFacing:Vector3f=&Lerp_Vector(
     A=&NormalizeVector(Vector To Normalize=&Subtract_Vector(A=$Particles.MeshOrigin, B=$Output.ConeConstraint_01.PositionToOrient))
     B=&Multiply_VectorByFloat(
      Vector=&TransformVector(Vector=&MakeVector(X=-1, Y=0, Z=-0.2), Destination Space=::World)
      Float=-1
     )
     Alpha=&MakeCustomFloatFromBool(Bool=$User.IsSkydiving, True Float=1, False Float=0)
    )) {isSetParameters=true}
   /SetVariables(Particles.MeshFacingAdd:Vector3f=&Lerp_Vector(
     A=&NormalizeVector(Vector To Normalize=&Subtract_Vector(A=$Particles.MeshOrigin, B=$Output.ConeConstraint_03.PositionToOrient))
     B=&Multiply_VectorByFloat(Vector=&TransformVector(Vector=<-1 0 -0.2>, Destination Space=::World), Float=-1)
     Alpha=&MakeCustomFloatFromBool(Bool=$User.IsSkydiving, True Float=1, False Float=0)
    )) {isSetParameters=true}
   /Niagara/Modules/Update/Orientation/V2/OrientMeshToVector(
    Look At Direction:Vector3f=&Multiply_VectorByFloat(Vector=$Particles.MeshFacing, Float=1)
    Override Incoming Orientation=1
    Override Quat=false
    Reference Axis Vector:Vector3f=<0 0 1>
    Up Axis Direction:Vector3f=&TransformVector(Vector=<0 0 1>, Destination Space=::World)
    UseOptionalSideAxis=false
    UseOptionalUpAxis=true
   )
   &OrientMeshToVector(
    Look At Direction:Vector3f=&Multiply_VectorByFloat(Vector=$Particles.MeshFacingAdd, Float=1)
    Reference Axis Vector:Vector3f=<0 0 1>
    Up Axis Direction:Vector3f=&TransformVector(Vector=<0 0 1>, Destination Space=::World)
    UseOptionalSideAxis=false
    UseOptionalUpAxis=true
    Write to Intrinsic Parameters=false
   ) {module=OrientMeshToVector002}
   &DynamicMaterialParameters(
    Index 0 Param 1=&MakeFloatFromVector(Vector=Linked() {linkedVariable {name=Particles.Previous.Position, type=NiagaraPosition}}, Channel=::NewEnumerator0)
    Index 0 Param 2=&MakeFloatFromVector(Vector=Linked() {linkedVariable {name=Particles.Previous.Position, type=NiagaraPosition}}, Channel=::NewEnumerator1)
    Index 0 Param 3=&MakeFloatFromVector(Vector=Linked() {linkedVariable {name=Particles.Previous.Position, type=NiagaraPosition}}, Channel=::NewEnumerator2)
    Index 1 Param 1=&MakeFloatFromVector(Vector=$Particles.HairDir, Channel=::NewEnumerator0)
    Index 1 Param 2=&MakeFloatFromVector(Vector=$Particles.HairDir, Channel=::NewEnumerator1)
    Index 1 Param 3=&MakeFloatFromVector(Vector=$Particles.HairDir, Channel=::NewEnumerator2)
    Index 2 Param 1=HlslExpression() {hlslExpression=Particles.MeshOrigin.x}
    Index 2 Param 2=HlslExpression() {hlslExpression=Particles.MeshOrigin.y}
    Index 2 Param 3=HlslExpression() {hlslExpression=Particles.MeshOrigin.z}
    Index 2 Param 4=$Particles.PendulumMaskFalloff
    Index 3 Param 1=&MakeFloatFromVector(Vector=Linked() {linkedVariable {name=Particles.Position, type=NiagaraPosition}}, Channel=::NewEnumerator0)
    Index 3 Param 2=&MakeFloatFromVector(Vector=Linked() {linkedVariable {name=Particles.Position, type=NiagaraPosition}}, Channel=::NewEnumerator1)
    Index 3 Param 3=&MakeFloatFromVector(Vector=Linked() {linkedVariable {name=Particles.Position, type=NiagaraPosition}}, Channel=::NewEnumerator2)
    Index 3 Param 4=$Particles.PendulumRigidity
    Module.Param0WriteEnabled=true
    Module.Param0WriteEnabled1=true
    Module.Param0WriteEnabled2=true
    Module.Param0WriteEnabled3=true
    Module.Param1WriteEnabled=true
    Module.Param1WriteEnabled1=true
    Module.Param1WriteEnabled2=true
    Module.Param1WriteEnabled3=true
    Module.Param2WriteEnabled=true
    Module.Param2WriteEnabled1=true
    Module.Param2WriteEnabled2=true
    Module.Param2WriteEnabled3=true
    Module.Param3WriteEnabled=true
    Module.Param3WriteEnabled1=true
    Module.Param3WriteEnabled2=true
    Module.Param3WriteEnabled3=true
    Write Parameter Index 1=true
    Write Parameter Index 2=true
    Write Parameter Index 3=true
   )
   /SetVariables(Particles.MeshOrientationAdd:Quat4f=$Output.OrientMeshToVector002.Raw Orientation From Vectors) {isSetParameters=true}
   /SetVariables(Particles.VisibilityTag=&MakeCustomIntFromBool(Bool=$User.bisHeadHidden, True INT=1, False INT=0)) {isSetParameters=true, module=SetVariables_#2}
  ]
  rendererClasses [&NiagaraMeshRendererProperties, &NiagaraSpriteRendererProperties]
  renderers [
   {rendererClass=&NiagaraMeshRendererProperties, rendererIndex=0}
   {rendererClass=&NiagaraSpriteRendererProperties, rendererIndex=1}
   {rendererClass=&NiagaraMeshRendererProperties, rendererIndex=2}
   {rendererClass=&NiagaraMeshRendererProperties, rendererIndex=3}
   {rendererClass=&NiagaraMeshRendererProperties, rendererIndex=4}
  ]
 }
 Light {
  emitterUpdateScript [
   &EmitterState(Life Cycle Mode=::NewEnumerator1, Loop Behavior=::NewEnumerator1, Loop Duration=1, UseLoopDelay=false)
   &SpawnBurst_Instantaneous(Use Spawn Probability=false, UseLoopCountLimit=false)
  ]
  particleSpawnScript [
   /Niagara/Modules/Spawn/Initialization/V2/InitializeParticle(Color=<rgba 0.6859 0 1.5 1>, Lifetime=0.1, Sprite Size Mode=::NewEnumerator3, UsePositionOffset=false, Write Material Random=true)
   &SystemLocation(MaskSpawnGroup=false, Offset=$System.HitOffset)
  ]
  particleUpdateScript [
   &ParticleState(Kill Particles When Lifetime Has Elapsed=false)
   /SetVariables(Particles.HairDir:Vector3f=&TransformVector(Vector=<-1 0 0>, Destination Space=::World)) {isSetParameters=true}
   /SetVariables(Particles.Position=&AddVectorToPosition(
     Position=$Engine.Emitter.SimulationPosition
     Vector=&Multiply_VectorByFloat(Vector=$Particles.HairDir, Float=20)
    )) {isSetParameters=true}
   &ScaleColor(
    Scale Alpha=&Lerp_Float(A=1, B=0, Alpha=$Particles.NormalizedAge)
    Scale RGB=&VectorFromFloat(Value=&Multiply_Float(A=1.5, B=$User.OpacityMult))
    ScaleA=false
    ScaleRGB=true
   )
   &FN_Light_v2(
    Light Exponent=3
    Light Radius=80
    LightEnabled:NiagaraBool=&Boolean_NotOperation(Bool=$User.bisHeadHidden)
    Scale Color=1
   )
   &CameraOffset(Camera Offset Amount=10)
  ]
  rendererClasses [&NiagaraLightRendererProperties]
  renderers [{rendererClass=&NiagaraLightRendererProperties, rendererIndex=0}]
 }
}
summary {
 systemName=NS_F_MED_GalaxyKnight_Head_Idle
 emitters [
  {emitterName=Bits, rendererClasses [&NiagaraSpriteRendererProperties]}
  {emitterName=Hair, rendererClasses [&NiagaraMeshRendererProperties, &NiagaraSpriteRendererProperties]}
  {emitterName=Light, rendererClasses [&NiagaraLightRendererProperties]}
 ]
 userVariables [
  User.bisHeadHidden {type=NiagaraBool}
  User.EmissiveMult {type=NiagaraFloat, defaultValue=1, description=""}
  User.FrontEnd {type=NiagaraBool, defaultValue=true}
  User.IsSkydiving {type=NiagaraBool}
  User.OpacityMult {type=NiagaraFloat, defaultValue=1, description=""}
  User.ParaGlideLeanAlpha {type=NiagaraFloat, defaultValue=0, description=""}
  User.SkeletalMesh {
   type=NiagaraDataInterfaceSkeletalMesh
   defaultValue {
    struct=NiagaraExt_VariableValue_DataInterface
    value {dataInterface=&self:NiagaraDataInterfaceSkeletalMesh_0}
   }
  }
 ]
}
