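# Anim Blueprint for the GalaxyKnight Ultra body cosmetic, adding procedural motion to its floating shoulder pieces.
# It copies the attached parent's pose, constrains shoulder base bones to arm bones, then runs AnimDynamics, faded out while parachuting.
# Phase-offset sine waves from real time bob the pieces additively before a final Control Rig pass.
# After evaluation, a zero-length delay precedes writing the chest glow socket's location to a material parameter.
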AnimBlueprint /BRCosmetics/Characters/Player/Male/Medium/Bodies/M_MED_GalaxyKnight_Ultra/Meshes/M_MED_GalaxyKnight_Ultra_AnimBP
target_skeleton=&/M_MED_GalaxyKnight_Ultra_Skeleton
graphs [
 AnimGraph {nodes [
   n1 = "Constraint - dyn_Shoulder_pices_01_base_r (PPPPP)"!(
    ComponentPose=Local To Component(LocalPose=Copy Pose From Mesh() {bUseAttachedParent=true})
    Alpha=1
    ConstraintWeights_0=0.34
    ConstraintWeights_1:float=0.34
    ConstraintWeights_2:float=0
    ConstraintWeights_3:float=0.32
    ConstraintWeights_4:float=0
   ) -> (Pose) {
    alphaBoolBlend {blendOption=Linear}
    boneToModify {boneName=dyn_Shoulder_pices_01_base_r}
    constraintSetup [
     {targetBone {boneName=upperarm_twist_02_r}}
     {targetBone {boneName=clavicle_r}}
     {targetBone {boneName=deform_upperarm_up_r}}
     {targetBone {boneName=deform_upperarm_bk_r}}
     {targetBone {boneName=deform_upperarm_fr_r}}
    ]
    constraintWeights [1,1,1,1,1]
   }
   n2 = "Constraint - dyn_Shoulder_pices_02_base_r (PP)"!(ComponentPose=@n1, Alpha=1, ConstraintWeights_0=0.9, ConstraintWeights_1:float=0.1) -> (Pose) {
    alphaBoolBlend {blendOption=Linear}
    boneToModify {boneName=dyn_Shoulder_pices_02_base_r}
    constraintSetup [{targetBone {boneName=upperarm_twist_02_r}}, {targetBone {boneName=upperarm_r}}]
    constraintWeights [1,1]
   }
   n3 = "Constraint - dyn_Shoulder_pices_03_base_r (PP)"!(ComponentPose=@n2, Alpha=1, ConstraintWeights_0=0.6, ConstraintWeights_1:float=0.4) -> (Pose) {
    alphaBoolBlend {blendOption=Linear}
    boneToModify {boneName=dyn_Shoulder_pices_03_base_r}
    constraintSetup [{targetBone {boneName=upperarm_twist_02_r}}, {targetBone {boneName=upperarm_r}}]
    constraintWeights [1,1]
   }
   n4 = "Constraint - dyn_Shoulder_pices_01_base_l (PPPPP)"!(
    ComponentPose=@n3
    Alpha=1
    ConstraintWeights_0=0.34
    ConstraintWeights_1:float=0.34
    ConstraintWeights_2:float=0
    ConstraintWeights_3:float=0.32
    ConstraintWeights_4:float=0
   ) -> (Pose) {
    alphaBoolBlend {blendOption=Linear}
    boneToModify {boneName=dyn_Shoulder_pices_01_base_l}
    constraintSetup [
     {targetBone {boneName=upperarm_twist_02_l}}
     {targetBone {boneName=clavicle_l}}
     {targetBone {boneName=deform_upperarm_up_l}}
     {targetBone {boneName=deform_upperarm_bk_l}}
     {targetBone {boneName=deform_upperarm_fr_l}}
    ]
    constraintWeights [1,1,1,1,1]
   }
   n5 = Get Alpha Parachute() -> (Alpha Parachute:double)
   n6 = Anim Dynamics - Bone: dyn_Shoulder_pices_01_r(ComponentPose=@n4, Alpha=@n5, ExternalForce="0,0,0") {
    angularDampingOverride=1
    angularSpringConstant=20000
    bAngularSpring=true
    bGravityOverrideInSimSpace=true
    bLinearSpring=true
    bOverrideAngularDamping=true
    bOverrideLinearDamping=true
    bUseGravityOverride=true
    bUseSphericalLimits=true
    linearDampingOverride=1
    linearSpringConstant=100
    simulationSpace=RootRelative
    alphaBoolBlend {blendOption=Linear}
    alphaScaleBiasClamp {scale=0.42}
    boundBone {boneName=dyn_Shoulder_pices_01_r}
    componentAppliedLinearAccClamp=<3000 3000 3000>
    componentLinearAccScale=<1 1 0.3>
    componentLinearVelScale=<0.7 0.7 0.5>
    gravityOverride {z=-100}
    physicsBodyDefinitions [{
      boundBone {boneName=dyn_Shoulder_pices_01_r}
      boxExtents=<10 10 10>
      constraintSetup {angularLimitsMax=<1 1 0>, angularLimitsMin=<-1 -1 0>, angularTarget=<1 0 0>, linearAxesMin=<0 0 -3>}
      localJointOffset=<0 0 0>
     }]
    relativeSpaceBone {boneName=upperarm_twist_02_l}
    sphericalLimits [{limitRadius=4, limitType=Inner, drivingBone {boneName=dyn_Shoulder_pices_01_base_r}, sphereLocalOffset=<0 0 0>}]
   }
   n7 = Anim Dynamics - Bone: dyn_Shoulder_pices_02_r(ComponentPose=@n6, Alpha=@n5, ExternalForce="0,0,0") {
    angularDampingOverride=1
    angularSpringConstant=20000
    bAngularSpring=true
    bGravityOverrideInSimSpace=true
    bLinearSpring=true
    bOverrideAngularDamping=true
    bOverrideLinearDamping=true
    bUseGravityOverride=true
    linearDampingOverride=1
    linearSpringConstant=150
    simulationSpace=RootRelative
    alphaBoolBlend {blendOption=Linear}
    alphaScaleBiasClamp {scale=0.23}
    boundBone {boneName=dyn_Shoulder_pices_02_r}
    componentAppliedLinearAccClamp=<2000 2000 2000>
    componentLinearAccScale=<0.5 0.5 0.2>
    componentLinearVelScale=<0.4 0.4 0.2>
    gravityOverride {z=1}
    physicsBodyDefinitions [{
      boundBone {boneName=dyn_Shoulder_pices_02_r}
      boxExtents=<10 10 3>
      constraintSetup {angularLimitsMax=<0 0 0>, angularLimitsMin=<0 0 0>, angularTarget=<1 0 0>, linearAxesMax=<3 3 0>, linearAxesMin=<-3 -3 -2>}
      localJointOffset=<0 0 0>
     }]
    relativeSpaceBone {boneName=upperarm_twist_02_l}
   }
   n8 = Anim Dynamics - Bone: dyn_Shoulder_pices_03_r(ComponentPose=@n7, Alpha=@n5, ExternalForce="0,0,0") {
    angularDampingOverride=1
    angularSpringConstant=20000
    bAngularSpring=true
    bGravityOverrideInSimSpace=true
    bLinearSpring=true
    bOverrideAngularDamping=true
    bOverrideLinearDamping=true
    bUseGravityOverride=true
    linearDampingOverride=1
    linearSpringConstant=60
    simulationSpace=RootRelative
    alphaBoolBlend {blendOption=Linear}
    alphaScaleBiasClamp {scale=0.46}
    boundBone {boneName=dyn_Shoulder_pices_03_r}
    componentAppliedLinearAccClamp=<3000 3000 3000>
    componentLinearAccScale=<0.3 0.3 0.1>
    componentLinearVelScale=<1 1 1.4>
    gravityOverride {z=1}
    physicsBodyDefinitions [{
      boundBone {boneName=dyn_Shoulder_pices_03_r}
      boxExtents=<10 10 3>
      constraintSetup {angularLimitsMax=<1 1 0>, angularLimitsMin=<-1 -1 0>, angularTarget=<1 0 0>, linearAxesMax=<2 3 0>, linearAxesMin=<-2 -3 -2>}
      localJointOffset=<0 0 0>
     }]
    relativeSpaceBone {boneName=upperarm_twist_02_l}
   }
   n9 = Anim Dynamics - Bone: dyn_Shoulder_pices_01_l(ComponentPose=@n8, Alpha=@n5, ExternalForce="0,0,0") {
    angularDampingOverride=1
    angularSpringConstant=20000
    bAngularSpring=true
    bGravityOverrideInSimSpace=true
    bLinearSpring=true
    bOverrideAngularDamping=true
    bOverrideLinearDamping=true
    bUseGravityOverride=true
    linearDampingOverride=1
    linearSpringConstant=100
    simulationSpace=RootRelative
    alphaBoolBlend {blendOption=Linear}
    alphaScaleBiasClamp {scale=0.52}
    boundBone {boneName=dyn_Shoulder_pices_01_l}
    componentAppliedLinearAccClamp=<3000 3000 3000>
    componentLinearAccScale=<1.2 1 0.5>
    componentLinearVelScale=<0.7 0.7 0.5>
    gravityOverride {z=-100}
    physicsBodyDefinitions [{
      boundBone {boneName=dyn_Shoulder_pices_01_l}
      boxExtents=<10 10 10>
      constraintSetup {angularLimitsMax=<1 1 0>, angularLimitsMin=<-1 -1 0>, angularTarget=<1 0 0>, linearAxesMax=<0 0 1>, linearAxesMin=<0 0 -1>}
      localJointOffset=<0 0 0>
     }]
    relativeSpaceBone {boneName=upperarm_twist_02_l}
   }
   n10 = Anim Dynamics - Bone: dyn_Shoulder_pices_02_l(ComponentPose=@n9, Alpha=@n5, ExternalForce="0,0,0") {
    angularDampingOverride=1
    angularSpringConstant=20000
    bAngularSpring=true
    bGravityOverrideInSimSpace=true
    bLinearSpring=true
    bOverrideAngularDamping=true
    bOverrideLinearDamping=true
    bUseGravityOverride=true
    linearDampingOverride=1
    linearSpringConstant=100
    simulationSpace=RootRelative
    alphaBoolBlend {blendOption=Linear}
    alphaScaleBiasClamp {scale=0.23}
    boundBone {boneName=dyn_Shoulder_pices_02_l}
    componentAppliedLinearAccClamp=<2000 2000 2000>
    componentLinearAccScale=<0.5 0.5 0.2>
    componentLinearVelScale=<0.4 0.4 0.2>
    gravityOverride {z=1}
    physicsBodyDefinitions [{
      boundBone {boneName=dyn_Shoulder_pices_02_l}
      boxExtents=<10 10 3>
      constraintSetup {angularLimitsMax=<1 1 0>, angularLimitsMin=<-1 -1 0>, angularTarget=<1 0 0>, linearAxesMax=<3 3 0>, linearAxesMin=<-3 -3 -2>}
      localJointOffset=<0 0 0>
     }]
    relativeSpaceBone {boneName=upperarm_twist_02_l}
   }
   n11 = Anim Dynamics - Bone: dyn_Shoulder_pices_03_l(ComponentPose=@n10, Alpha=@n5, ExternalForce="0,0,0") {
    angularDampingOverride=1
    angularSpringConstant=20000
    bAngularSpring=true
    bGravityOverrideInSimSpace=true
    bLinearSpring=true
    bOverrideAngularDamping=true
    bOverrideLinearDamping=true
    bUseGravityOverride=true
    linearDampingOverride=1
    linearSpringConstant=60
    simulationSpace=RootRelative
    alphaBoolBlend {blendOption=Linear}
    alphaScaleBiasClamp {scale=0.46}
    boundBone {boneName=dyn_Shoulder_pices_03_l}
    componentAppliedLinearAccClamp=<3000 3000 3000>
    componentLinearAccScale=<0.3 0.3 0.1>
    componentLinearVelScale=<1 1 1.4>
    gravityOverride {z=1}
    physicsBodyDefinitions [{
      boundBone {boneName=dyn_Shoulder_pices_03_l}
      boxExtents=<10 10 3>
      constraintSetup {angularLimitsMax=<1 1 0>, angularLimitsMin=<-1 -1 0>, angularTarget=<1 0 0>, linearAxesMax=<2 3 0>, linearAxesMin=<-2 -3 -2>}
      localJointOffset=<0 0 0>
     }]
    relativeSpaceBone {boneName=upperarm_twist_02_l}
   }
   n12 = Get Wave_01() -> (Wave_01:double)
   n13 = "Transform (Modify) Bone - Bone: dyn_Shoulder_pices_01_l"!(ComponentPose=@n11, Alpha=1, Translation_X:double=0, Translation_Y:double=0, Translation_Z:double=@n12, Rotation="0,0,0", Scale="1,1,1") -> (Pose) {
    translationMode=BMM_Additive
    translationSpace=BCS_BoneSpace
    alphaBoolBlend {blendOption=Linear}
    boneToModify {boneName=dyn_Shoulder_pices_01_l}
    rotation {roll=25, yaw=15}
   }
   n14 = Get Wave_02() -> (Wave_02:double)
   n15 = "Transform (Modify) Bone - Bone: dyn_Shoulder_pices_02_l"!(ComponentPose=@n13, Alpha=1, Translation_X:double=0, Translation_Y:double=0, Translation_Z:double=@n14, Rotation="0,0,0", Scale="1,1,1") -> (Pose) {
    translationMode=BMM_Additive
    translationSpace=BCS_BoneSpace
    alphaBoolBlend {blendOption=Linear}
    boneToModify {boneName=dyn_Shoulder_pices_02_l}
    rotation {roll=25, yaw=15}
   }
   n16 = Get Wave_03() -> (Wave_03:double)
   n17 = "Transform (Modify) Bone - Bone: dyn_Shoulder_pices_03_l"!(ComponentPose=@n15, Alpha=1, Translation_X:double=0, Translation_Y:double=0, Translation_Z:double=@n16, Rotation="0,0,0", Scale="1,1,1") -> (Pose) {
    translationMode=BMM_Additive
    translationSpace=BCS_BoneSpace
    alphaBoolBlend {blendOption=Linear}
    boneToModify {boneName=dyn_Shoulder_pices_03_l}
    rotation {roll=25, yaw=15}
   }
   n18 = "Transform (Modify) Bone - Bone: dyn_Shoulder_pices_01_r"!(ComponentPose=@n17, Alpha=1, Translation_X:double=0, Translation_Y:double=0, Translation_Z:double=@n12, Rotation="0,0,0", Scale="1,1,1") -> (Pose) {
    translationMode=BMM_Additive
    translationSpace=BCS_BoneSpace
    alphaBoolBlend {blendOption=Linear}
    boneToModify {boneName=dyn_Shoulder_pices_01_r}
    rotation {roll=25, yaw=15}
   }
   n19 = "Transform (Modify) Bone - Bone: dyn_Shoulder_pices_02_r"!(ComponentPose=@n18, Alpha=1, Translation_X:double=0, Translation_Y:double=0, Translation_Z:double=@n14, Rotation="0,0,0", Scale="1,1,1") -> (Pose) {
    translationMode=BMM_Additive
    translationSpace=BCS_BoneSpace
    alphaBoolBlend {blendOption=Linear}
    boneToModify {boneName=dyn_Shoulder_pices_02_r}
    rotation {roll=25, yaw=15}
   }
   n20 = "Transform (Modify) Bone - Bone: dyn_Shoulder_pices_03_r"!(ComponentPose=@n19, Alpha=1, Translation_X:double=0, Translation_Y:double=0, Translation_Z:double=@n16, Rotation="0,0,0", Scale="1,1,1") -> (Pose) {
    translationMode=BMM_Additive
    translationSpace=BCS_BoneSpace
    alphaBoolBlend {blendOption=Linear}
    boneToModify {boneName=dyn_Shoulder_pices_03_r}
    rotation {roll=25, yaw=15}
   }
   n21 = Control Rig(Source=Component To Local(ComponentPose=@n20)) {
    alphaBoolBlend {blendOption=Linear}
    controlRigAssetReference {blueprintRigClass=&/M_MED_GalaxyKnight_Ultra_ControlRig.M_MED_GalaxyKnight_Ultra_ControlRig_C}
   }
   Output Pose(Result=@n21) {layerGroup=None}
  ]}
 EventGraph {nodes [
   n1 = Event BlueprintUpdateAnimation()
   n2 = GetRealTimeSeconds()
   n3 = float + float(A=float * float(A=@n2, B=1.5) {is=title K2Node_PromotableOperator}, B=0.6)
   > Set Wave_01(Wave_01:double=float * float(A="Sin (Radians)"(A=@n3), B=0.3)) -> (Output_Get:double)
   n4 = float + float(A=float * float(A=@n2, B=1.5), B=0.8)
   > Set Wave_02(Wave_02:double=float * float(A="Sin (Radians)"(A=@n4), B=0.3)) -> (Output_Get:double)
   n5 = float * float(A=@n2, B=1.5)
   > Set Wave_03(Wave_03:double=float * float(A="Sin (Radians)"(A=@n5), B=0.3)) -> (Output_Get:double)
   > n6 = Branch(Condition=NOT Boolean(A=Get bIsParachuteOpened()))
   > Set Alpha Parachute(Alpha Parachute:double=FInterpTo(Current=Get Alpha Parachute(), Target=1, DeltaTime=@n1.DeltaTimeX, InterpSpeed=12)) -> (Output_Get:double)
   Set Alpha Parachute(execute=@n6.else, Alpha Parachute=FInterpTo(Current=Get Alpha Parachute(), Target=0, DeltaTime=@n1.DeltaTimeX, InterpSpeed=12))
   Event BlueprintPostEvaluateAnimation() {is=title K2Node_Event}
   n7 = GetPartSkeletalMeshComponent()
   > Is Valid(InputObject=@n7)
   > Delay(Duration=0)
   > SetVectorParameterValueOnMaterials!(
    execute
    self:Mesh Component Object Reference=@n7
    ParameterName:Name=GK_ChestGlowLocation
    ParameterValue:Vector=GetSocketLocation!(self:Scene Component Object Reference=@n7, InSocketName:Name=ChestGlowSocket) -> (ReturnValue:Vector)
   ) -> (then)
  ]}
]
variables [
 Alpha Parachute {type=number}
 Wave_01 {type=number}
 Wave_02 {type=number}
 Wave_03 {type=number}
]
