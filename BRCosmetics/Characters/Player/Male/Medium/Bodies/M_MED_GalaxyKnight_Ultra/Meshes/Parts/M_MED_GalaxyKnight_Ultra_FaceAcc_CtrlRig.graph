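# Control Rig for the Galaxy Knight Ultra face accessory part, applying pose-driven corrective offsets to the helmet neck plates and head.
# A spherical pose reader measures head orientation relative to neck_01 within an asymmetric active and falloff region.
# Its output weights three chained Modify Transforms nodes that offset both dyn_neck_plates bones and slightly rotate the head.
# Crown piece and helmet FX bones are in the hierarchy but untouched by the graph.
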ControlRigBlueprint /BRCosmetics/Characters/Player/Male/Medium/Bodies/M_MED_GalaxyKnight_Ultra/Meshes/Parts/M_MED_GalaxyKnight_Ultra_FaceAcc_CtrlRig
graphs [
 RigVMFunctionLibrary {links [], nodes []}
 RigVMModel {nodes [
   RigUnit_BeginExecution = &RigUnit_BeginExecution()
   n1 = &RigUnit_SphericalPoseReader(
    ActiveRegionScaleFactors=(PositiveWidth=1,NegativeWidth=1,PositiveHeight=0.6,NegativeHeight=1)
    ActiveRegionSize=0.1
    Debug=(bDrawDebug=false,bDrawLocalAxes=true,DebugScale=25,DebugSegments=20,DebugThickness=0.25)
    DriverAxis=(X=1,Y=0,Z=0)
    DriverItem=(Type=Bone,Name="head")
    FalloffRegionScaleFactors=(PositiveWidth=1,NegativeWidth=1,PositiveHeight=0.6,NegativeHeight=1)
    FalloffSize=0.2
    OptionalParentItem=(Type=Bone,Name="neck_01")
    RotationOffset=(X=-30,Y=-80,Z=0)
    ExecutePin=@RigUnit_BeginExecution.ExecutePin
   )
   > &RigUnit_ModifyTransforms(
    ItemToModify=((Item=(Type=Bone,Name="dyn_neck_plates_02"),Transform=(Rotation=(X=0,Y=0,Z=0,W=1),Translation=(X=0,Y=0,Z=0),Scale3D=(X=1,Y=1,Z=1))))
    ItemToModify.0.Transform=/Script/RigVM.RigVMFunction_MathTransformFromSRT(Location=(X=1,Y=1,Z=1), Rotation=(X=-5,Y=-20,Z=10), Scale=(X=1,Y=1,Z=1)) {is=struct RigVMUnitNode}.Transform
    Weight=@n1.OutputParam
   )
   > &RigUnit_ModifyTransforms(
    ItemToModify=((Item=(Type=Bone,Name="dyn_neck_plates_01"),Transform=(Rotation=(X=0,Y=0,Z=0,W=1),Translation=(X=0,Y=0,Z=0),Scale3D=(X=1,Y=1,Z=1))))
    ItemToModify.0.Transform=&RigVMFunction_MathTransformFromSRT(Location=(X=0.5,Y=0.3,Z=0.5), Rotation=(X=-5,Y=-10,Z=10), Scale=(X=1,Y=1,Z=1)).Transform
    Weight=@n1.OutputParam
   )
   > &RigUnit_ModifyTransforms(
    ItemToModify=((Item=(Type=Bone,Name="head"),Transform=(Rotation=(X=0,Y=0,Z=0,W=1),Translation=(X=0,Y=0,Z=0),Scale3D=(X=1,Y=1,Z=1))))
    ItemToModify.0.Transform=&RigVMFunction_MathTransformFromSRT(Location=(X=0,Y=0,Z=0), Rotation=(X=0,Y=5,Z=3), Scale=(X=1,Y=1,Z=1)).Transform
    Weight=@n1.OutputParam
   )
  ]}
]
hierarchy [
 root {loc=<0 0 0>, rot=<0 0 0 1>}
 pelvis {parent=root, loc=<0 -0.7071 93.1191>, rot=<0.0127 0.707 -0.0127 -0.707>}
 spine_01 {parent=pelvis, loc=<6.7494 -0.6388 0>, rot=<0 0 0.034 -0.9994>}
 spine_02 {parent=spine_01, loc=<7.2518 -0.035 0>, rot=<0 0 0.0238 -0.9997>}
 spine_03 {parent=spine_02, loc=<6.0104 0 0>, rot=<0 0 0.0149 -0.9999>}
 spine_04 {parent=spine_03, loc=<8.1024 0 0>, rot=<0 0 -0.0132 -0.9999>}
 spine_05 {parent=spine_04, loc=<10.0146 0 0>, rot=<0 0 -0.0015 -1>}
 neck_01 {parent=spine_05, loc=<11.3138 0.1529 0>, rot=<0 0 -0.2003 -0.9797>}
 neck_02 {parent=neck_01, loc=<4.5627 0 0>, rot=<0 0 0.0504 -0.9987>}
 head {parent=neck_02, loc=<5.3914 0 0>, rot=<0 0 0.115 -0.9934>}
 dyn_neck_plates_base_01 {parent=head, loc=<8.4302 -8.4874 -0.1412>, rot=<0.5913 -0.3878 0.3878 -0.5913>}
 dyn_neck_plates_01 {parent=dyn_neck_plates_base_01, loc=<0 0 0>, rot=<0 0 0 1>}
 dyn_neck_plates_base_02 {parent=head, loc=<4.842 -10.2495 0.0299>, rot=<0.5913 -0.3878 0.3878 -0.5913>}
 dyn_neck_plates_02 {parent=dyn_neck_plates_base_02, loc=<0 0 0>, rot=<0 0 0 1>}
 dyn_helmet_crown_piece_base_01 {parent=head, loc=<22.4031 0.4693 -2.596>, rot=<-0.0812 -0.8238 -0.0703 -0.5566>}
 dyn_helmet_crown_piece_01 {parent=dyn_helmet_crown_piece_base_01, loc=<0 0 0>, rot=<0 0 0 1>}
 dyn_helmet_crown_piece_base_02 {parent=head, loc=<21.4509 2.4558 -0.992>, rot=<-0.3583 -0.6486 -0.3744 -0.5575>}
 dyn_helmet_crown_piece_02 {parent=dyn_helmet_crown_piece_base_02, loc=<0 0 0>, rot=<0 0 0 1>}
 dyn_helmet_crown_piece_base_03 {parent=head, loc=<22.5195 2.7406 0.725>, rot=<0.5462 0.3793 0.6356 0.3923>}
 dyn_helmet_crown_piece_03 {parent=dyn_helmet_crown_piece_base_03, loc=<0 0 0>, rot=<0 0 0 1>}
 dyn_helmet_crown_piece_base_04 {parent=head, loc=<21.39 1.4647 2.071>, rot=<0.5986 0.1109 0.7733 0.1771>}
 dyn_helmet_crown_piece_04 {parent=dyn_helmet_crown_piece_base_04, loc=<0 0 0>, rot=<0 0 0 1>}
 dyn_helmet_crown_piece_base_05 {parent=head, loc=<20.7306 -0.9696 1.4759>, rot=<0.5305 -0.1876 0.8069 -0.1799>}
 dyn_helmet_crown_piece_05 {parent=dyn_helmet_crown_piece_base_05, loc=<0 0 0>, rot=<0 0 0 1>}
 dyn_helmet_crown_piece_base_06 {parent=head, loc=<21.0301 -0.6238 -1.8906>, rot=<0.098 -0.8216 0.1951 -0.5266>}
 dyn_helmet_crown_piece_06 {parent=dyn_helmet_crown_piece_base_06, loc=<0 0 0>, rot=<0 0 0 1>}
 fx_helmet {parent=head, loc=<18.1361 0.7448 -0.2511>, rot=<0.0029 -0.7071 -0.0029 -0.7071>}
 clavicle_l {parent=spine_05, loc=<4.3575 -0.1466 -2.3223>, rot=<0.0465 -0.7289 -0.009 -0.6829>}
 clavicle_r {parent=spine_05, loc=<4.3575 -0.1466 2.3223>, rot=<0.7289 0.0465 -0.6829 0.009>}
]
