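# Anim Blueprint for the GalaxyKnight Ultra face accessory, animating six floating helmet crown pieces and two neck plates.
# Update event writes phase-offset real-time sine values that additively shift each bone along local X, with neck plate variables swapped between bones.
# The graph copies the parent mesh pose, applies those shifts, then AnimDynamics and a Control Rig.
# AnimDynamics alpha eases out while the parachute is open and back in otherwise.
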AnimBlueprint /BRCosmetics/Characters/Player/Male/Medium/Bodies/M_MED_GalaxyKnight_Ultra/Meshes/Parts/M_MED_GalaxyKnight_Ultra_FaceAcc_AnimBP
target_skeleton=&/M_MED_GalaxyKnight_Ultra_FaceAcc_Skeleton
graphs [
 AnimGraph {nodes [
   n1 = "Transform (Modify) Bone - Bone: dyn_helmet_crown_piece_01"!(
    ComponentPose=Local To Component(LocalPose=Copy Pose From Mesh() {bUseAttachedParent=true})
    Alpha=1
    Translation_X:double=Get Crown_Piece_01() -> (Crown_Piece_01:double)
    Translation_Y:double=0
    Translation_Z:double=0
    Rotation="0,0,0"
    Scale="1,1,1"
   ) -> (Pose) {
    translationMode=BMM_Additive
    translationSpace=BCS_BoneSpace
    alphaBoolBlend {blendOption=Linear}
    boneToModify {boneName=dyn_helmet_crown_piece_01}
   }
   n2 = "Transform (Modify) Bone - Bone: dyn_helmet_crown_piece_02"!(
    ComponentPose=@n1
    Alpha=1
    Translation_X:double=Get Crown_Piece_02() -> (Crown_Piece_02:double)
    Translation_Y:double=0
    Translation_Z:double=0
    Rotation="0,0,0"
    Scale="1,1,1"
   ) -> (Pose) {
    translationMode=BMM_Additive
    translationSpace=BCS_BoneSpace
    alphaBoolBlend {blendOption=Linear}
    boneToModify {boneName=dyn_helmet_crown_piece_02}
   }
   n3 = "Transform (Modify) Bone - Bone: dyn_helmet_crown_piece_03"!(
    ComponentPose=@n2
    Alpha=1
    Translation_X:double=Get Crown_Piece_03() -> (Crown_Piece_03:double)
    Translation_Y:double=0
    Translation_Z:double=0
    Rotation="0,0,0"
    Scale="1,1,1"
   ) -> (Pose) {
    translationMode=BMM_Additive
    translationSpace=BCS_BoneSpace
    alphaBoolBlend {blendOption=Linear}
    boneToModify {boneName=dyn_helmet_crown_piece_03}
   }
   n4 = "Transform (Modify) Bone - Bone: dyn_helmet_crown_piece_04"!(
    ComponentPose=@n3
    Alpha=1
    Translation_X:double=Get Crown_Piece_04() -> (Crown_Piece_04:double)
    Translation_Y:double=0
    Translation_Z:double=0
    Rotation="0,0,0"
    Scale="1,1,1"
   ) -> (Pose) {
    translationMode=BMM_Additive
    translationSpace=BCS_BoneSpace
    alphaBoolBlend {blendOption=Linear}
    boneToModify {boneName=dyn_helmet_crown_piece_04}
   }
   n5 = "Transform (Modify) Bone - Bone: dyn_helmet_crown_piece_05"!(
    ComponentPose=@n4
    Alpha=1
    Translation_X:double=Get Crown_Piece_05() -> (Crown_Piece_05:double)
    Translation_Y:double=0
    Translation_Z:double=0
    Rotation="0,0,0"
    Scale="1,1,1"
   ) -> (Pose) {
    translationMode=BMM_Additive
    translationSpace=BCS_BoneSpace
    alphaBoolBlend {blendOption=Linear}
    boneToModify {boneName=dyn_helmet_crown_piece_05}
   }
   n6 = "Transform (Modify) Bone - Bone: dyn_helmet_crown_piece_06"!(
    ComponentPose=@n5
    Alpha=1
    Translation_X:double=Get Crown_Piece_06() -> (Crown_Piece_06:double)
    Translation_Y:double=0
    Translation_Z:double=0
    Rotation="0,0,0"
    Scale="1,1,1"
   ) -> (Pose) {
    translationMode=BMM_Additive
    translationSpace=BCS_BoneSpace
    alphaBoolBlend {blendOption=Linear}
    boneToModify {boneName=dyn_helmet_crown_piece_06}
   }
   n7 = "Transform (Modify) Bone - Bone: dyn_neck_plates_01"!(
    ComponentPose=@n6
    Alpha=1
    Translation_X:double=Get Neck_Plates_02() -> (Neck_Plates_02:double)
    Translation_Y:double=0
    Translation_Z:double=0
    Rotation="0,0,0"
    Scale="1,1,1"
   ) -> (Pose) {
    translationMode=BMM_Additive
    translationSpace=BCS_BoneSpace
    alphaBoolBlend {blendOption=Linear}
    boneToModify {boneName=dyn_neck_plates_01}
   }
   n8 = "Transform (Modify) Bone - Bone: dyn_neck_plates_02"!(
    ComponentPose=@n7
    Alpha=1
    Translation_X:double=Get Neck_Plates_01() -> (Neck_Plates_01:double)
    Translation_Y:double=0
    Translation_Z:double=0
    Rotation="0,0,0"
    Scale="1,1,1"
   ) -> (Pose) {
    translationMode=BMM_Additive
    translationSpace=BCS_BoneSpace
    alphaBoolBlend {blendOption=Linear}
    boneToModify {boneName=dyn_neck_plates_02}
   }
   n9 = Get Alpha parachute() -> (Alpha parachute:double)
   n10 = Anim Dynamics - Bone: dyn_neck_plates_01(ComponentPose=@n8, Alpha=@n9, ExternalForce="0,0,0") {
    angularBiasOverride=1
    angularDampingOverride=1
    angularSpringConstant=20000
    bAngularSpring=true
    bOverrideLinearDamping=true
    bUseGravityOverride=true
    linearDampingOverride=1
    linearSpringConstant=100
    simulationSpace=RootRelative
    alphaBoolBlend {blendOption=Linear}
    alphaScaleBiasClamp {scale=0.2}
    boundBone {boneName=dyn_neck_plates_01}
    componentAppliedLinearAccClamp=<3000 3000 3000>
    componentLinearAccScale=<0.3 0.3 0.1>
    componentLinearVelScale=<0.7 0.7 0.5>
    gravityOverride {z=1}
    physicsBodyDefinitions [{
      collisionType=CustomSphere
      sphereCollisionRadius=1.6
      boundBone {boneName=dyn_neck_plates_01}
      boxExtents=<100 100 100>
      constraintSetup {angularLimitsMax=<0 0 0>, angularLimitsMin=<0 0 0>, angularTarget=<0 0 0>, linearAxesMax=<0 1 1>, linearAxesMin=<0 -1 -1>}
      localJointOffset=<0 0 0.2>
     }]
    relativeSpaceBone {boneName=head}
   }
   n11 = Anim Dynamics - Bone: dyn_neck_plates_02(ComponentPose=@n10, Alpha=@n9, ExternalForce="0,0,0") {
    angularDampingOverride=1
    angularSpringConstant=20000
    bAngularSpring=true
    bLinearSpring=true
    bOverrideLinearDamping=true
    bUseGravityOverride=true
    linearDampingOverride=1
    linearSpringConstant=100
    simulationSpace=RootRelative
    alphaBoolBlend {blendOption=Linear}
    alphaScaleBiasClamp {scale=0.3}
    boundBone {boneName=dyn_neck_plates_02}
    componentAppliedLinearAccClamp=<3000 3000 3000>
    componentLinearAccScale=<0.3 0.3 0.1>
    componentLinearVelScale=<0.7 0.7 0.5>
    gravityOverride {z=1}
    physicsBodyDefinitions [{
      collisionType=CustomSphere
      sphereCollisionRadius=1.6
      boundBone {boneName=dyn_neck_plates_02}
      boxExtents=<100 100 100>
      constraintSetup {angularLimitsMax=<0 0 0>, angularLimitsMin=<0 0 0>, angularTarget=<0 0 0>, linearAxesMax=<0 1 1>, linearAxesMin=<0 -1 -1>}
      localJointOffset=<0 0 0>
     }]
    relativeSpaceBone {boneName=head}
   }
   n12 = Anim Dynamics - Bone: dyn_helmet_crown_piece_01(ComponentPose=@n11, Alpha=@n9, ExternalForce="0,0,0") {
    angularBiasOverride=1
    angularDampingOverride=1
    angularSpringConstant=20000
    bAngularSpring=true
    bOverrideLinearDamping=true
    bUseGravityOverride=true
    linearDampingOverride=1
    linearSpringConstant=300
    alphaBoolBlend {blendOption=Linear}
    alphaScaleBiasClamp {scale=0.45}
    boundBone {boneName=dyn_helmet_crown_piece_01}
    componentAppliedLinearAccClamp=<3000 3000 3000>
    componentLinearAccScale=<0.2 0.2>
    componentLinearVelScale=<0.7 0.7 0.5>
    gravityOverride {z=100}
    physicsBodyDefinitions [{
      collisionType=CustomSphere
      sphereCollisionRadius=1.2
      boundBone {boneName=dyn_helmet_crown_piece_01}
      boxExtents=<100 100 100>
      constraintSetup {angularLimitsMax=<0 0 0>, angularLimitsMin=<0 0 0>, angularTarget=<0 0 0>, linearAxesMax=<1.5 1.5 1.5>, linearAxesMin=<-1 -1.5 -1.5>}
      localJointOffset=<0 0 0>
     }]
    relativeSpaceBone {boneName=head}
   }
   n13 = Anim Dynamics - Bone: dyn_helmet_crown_piece_02(ComponentPose=@n12, Alpha=@n9, ExternalForce="0,0,0") {
    angularDampingOverride=1
    angularSpringConstant=20000
    bAngularSpring=true
    bLinearSpring=true
    bOverrideLinearDamping=true
    bUseGravityOverride=true
    linearDampingOverride=1
    linearSpringConstant=300
    alphaBoolBlend {blendOption=Linear}
    alphaScaleBiasClamp {scale=0.48}
    boundBone {boneName=dyn_helmet_crown_piece_02}
    componentAppliedLinearAccClamp=<3000 3000 3000>
    componentLinearAccScale=<0.2 0.2>
    componentLinearVelScale=<0.7 0.7 0.5>
    gravityOverride {z=100}
    physicsBodyDefinitions [{
      collisionType=CustomSphere
      sphereCollisionRadius=1.2
      boundBone {boneName=dyn_helmet_crown_piece_02}
      boxExtents=<100 100 100>
      constraintSetup {angularLimitsMax=<0 0 0>, angularLimitsMin=<0 0 0>, angularTarget=<0 0 0>, linearAxesMax=<1.5 1.5 1.5>, linearAxesMin=<-1 -1.5 -1.5>}
      localJointOffset=<0 0 0>
     }]
    relativeSpaceBone {boneName=head}
   }
   n14 = Anim Dynamics - Bone: dyn_helmet_crown_piece_03(ComponentPose=@n13, Alpha=@n9, ExternalForce="0,0,0") {
    angularBiasOverride=1
    angularDampingOverride=1
    angularSpringConstant=20000
    bAngularSpring=true
    bOverrideLinearDamping=true
    bUseGravityOverride=true
    linearDampingOverride=1
    linearSpringConstant=300
    alphaBoolBlend {blendOption=Linear}
    alphaScaleBiasClamp {scale=0.44}
    boundBone {boneName=dyn_helmet_crown_piece_03}
    componentAppliedLinearAccClamp=<3000 3000 3000>
    componentLinearAccScale=<0.2 0.2>
    componentLinearVelScale=<0.7 0.7 0.5>
    gravityOverride {z=100}
    physicsBodyDefinitions [{
      collisionType=CustomSphere
      sphereCollisionRadius=1.2
      boundBone {boneName=dyn_helmet_crown_piece_03}
      boxExtents=<100 100 100>
      constraintSetup {angularLimitsMax=<0 0 0>, angularLimitsMin=<0 0 0>, angularTarget=<0 0 0>, linearAxesMax=<1.5 1.5 1.5>, linearAxesMin=<-1 -1.5 -1.5>}
      localJointOffset=<0 0 0>
     }]
    relativeSpaceBone {boneName=head}
   }
   n15 = Anim Dynamics - Bone: dyn_helmet_crown_piece_04(ComponentPose=@n14, Alpha=@n9, ExternalForce="0,0,0") {
    angularDampingOverride=1
    angularSpringConstant=20000
    bAngularSpring=true
    bLinearSpring=true
    bOverrideLinearDamping=true
    bUseGravityOverride=true
    linearDampingOverride=1
    linearSpringConstant=300
    alphaBoolBlend {blendOption=Linear}
    alphaScaleBiasClamp {scale=0.52}
    boundBone {boneName=dyn_helmet_crown_piece_04}
    componentAppliedLinearAccClamp=<3000 3000 3000>
    componentLinearAccScale=<0.2 0.2>
    componentLinearVelScale=<0.7 0.7 0.5>
    gravityOverride {z=100}
    physicsBodyDefinitions [{
      collisionType=CustomSphere
      sphereCollisionRadius=1.2
      boundBone {boneName=dyn_helmet_crown_piece_04}
      boxExtents=<100 100 100>
      constraintSetup {angularLimitsMax=<0 0 0>, angularLimitsMin=<0 0 0>, angularTarget=<0 0 0>, linearAxesMax=<1.5 1.5 1.5>, linearAxesMin=<-1 -1.5 -1.5>}
      localJointOffset=<0 0 0>
     }]
    relativeSpaceBone {boneName=head}
   }
   n16 = Anim Dynamics - Bone: dyn_helmet_crown_piece_05(ComponentPose=@n15, Alpha=@n9, ExternalForce="0,0,0") {
    angularBiasOverride=1
    angularDampingOverride=1
    angularSpringConstant=20000
    bAngularSpring=true
    bOverrideLinearDamping=true
    bUseGravityOverride=true
    linearDampingOverride=1
    linearSpringConstant=300
    alphaBoolBlend {blendOption=Linear}
    alphaScaleBiasClamp {scale=0.4}
    boundBone {boneName=dyn_helmet_crown_piece_05}
    componentAppliedLinearAccClamp=<3000 3000 3000>
    componentLinearAccScale=<0.2 0.2>
    componentLinearVelScale=<0.7 0.7 0.5>
    gravityOverride {z=100}
    physicsBodyDefinitions [{
      collisionType=CustomSphere
      sphereCollisionRadius=1.2
      boundBone {boneName=dyn_helmet_crown_piece_05}
      boxExtents=<100 100 100>
      constraintSetup {angularLimitsMax=<0 0 0>, angularLimitsMin=<0 0 0>, angularTarget=<0 0 0>, linearAxesMax=<1.5 1.5 1.5>, linearAxesMin=<-1 -1.5 -1.5>}
      localJointOffset=<0 0 0>
     }]
    relativeSpaceBone {boneName=head}
   }
   n17 = Anim Dynamics - Bone: dyn_helmet_crown_piece_06(ComponentPose=@n16, Alpha=@n9, ExternalForce="0,0,0") {
    angularDampingOverride=1
    angularSpringConstant=20000
    bAngularSpring=true
    bLinearSpring=true
    bOverrideLinearDamping=true
    bUseGravityOverride=true
    linearDampingOverride=1
    linearSpringConstant=300
    alphaBoolBlend {blendOption=Linear}
    alphaScaleBiasClamp {scale=0.48}
    boundBone {boneName=dyn_helmet_crown_piece_06}
    componentAppliedLinearAccClamp=<3000 3000 3000>
    componentLinearAccScale=<0.2 0.2>
    componentLinearVelScale=<0.7 0.7 0.5>
    gravityOverride {z=100}
    physicsBodyDefinitions [{
      collisionType=CustomSphere
      sphereCollisionRadius=1.2
      boundBone {boneName=dyn_helmet_crown_piece_06}
      boxExtents=<100 100 100>
      constraintSetup {angularLimitsMax=<0 0 0>, angularLimitsMin=<0 0 0>, angularTarget=<0 0 0>, linearAxesMax=<1.5 1.5 1.5>, linearAxesMin=<-1 -1.5 -1.5>}
      localJointOffset=<0 0 0>
     }]
    relativeSpaceBone {boneName=head}
   }
   n18 = Control Rig(Source=Component To Local(ComponentPose=@n17)) {
    alphaBoolBlend {blendOption=Linear}
    controlRigAssetReference {blueprintRigClass=&/M_MED_GalaxyKnight_Ultra_FaceAcc_CtrlRig.M_MED_GalaxyKnight_Ultra_FaceAcc_CtrlRig_C}
   }
   Output Pose(Result=@n18) {layerGroup=None}
  ]}
 EventGraph {nodes [
   n1 = Event BlueprintUpdateAnimation()
   n2 = GetRealTimeSeconds()
   n3 = float + float(A=float * float(A=@n2, B=1) {is=title K2Node_PromotableOperator}, B=73)
   > Set Crown_Piece_01(Crown_Piece_01:double=float * float(A="Sin (Radians)"(A=@n3), B=0.6)) -> (Output_Get:double)
   n4 = float + float(A=float * float(A=@n2, B=1), B=42)
   > Set Crown_Piece_02(Crown_Piece_02:double=float * float(A="Sin (Radians)"(A=@n4), B=0.6)) -> (Output_Get:double)
   n5 = float + float(A=float * float(A=@n2, B=1), B=21)
   > Set Crown_Piece_03(Crown_Piece_03:double=float * float(A="Sin (Radians)"(A=@n5), B=0.7)) -> (Output_Get:double)
   n6 = float + float(A=float * float(A=@n2, B=1), B=3)
   > Set Crown_Piece_04(Crown_Piece_04:double=float * float(A="Sin (Radians)"(A=@n6), B=0.6)) -> (Output_Get:double)
   n7 = float + float(A=float * float(A=@n2, B=1), B=7)
   > Set Crown_Piece_05(Crown_Piece_05:double=float * float(A="Sin (Radians)"(A=@n7), B=0.7)) -> (Output_Get:double)
   n8 = float + float(A=float * float(A=@n2, B=1), B=15)
   > Set Crown_Piece_06(Crown_Piece_06:double=float * float(A="Sin (Radians)"(A=@n8), B=0.6)) -> (Output_Get:double)
   n9 = float + float(A=float * float(A=@n2, B=1), B=72)
   > Set Neck_Plates_01(Neck_Plates_01:double=float * float(A="Sin (Radians)"(A=@n9), B=0.7)) -> (Output_Get:double)
   n10 = float + float(A=float * float(A=@n2, B=1), B=18)
   > Set Neck_Plates_02(Neck_Plates_02:double=float * float(A="Sin (Radians)"(A=@n10), B=0.3)) -> (Output_Get:double)
   > n11 = Branch(Condition=NOT Boolean(A=Get bIsParachuteOpened()))
   > Set Alpha Parachute(Alpha Parachute:double=FInterpTo(Current=Get Alpha Parachute() -> (Alpha Parachute:double), Target=1, DeltaTime=@n1.DeltaTimeX, InterpSpeed=12)) -> (Output_Get:double)
   Set Alpha Parachute(execute=@n11.else, Alpha Parachute=FInterpTo(Current=Get Alpha Parachute(), Target=0, DeltaTime=@n1.DeltaTimeX, InterpSpeed=12))
  ]}
]
variables [
 Alpha parachute {type=number}
 Crown_Piece_01 {type=number}
 Crown_Piece_02 {type=number}
 Crown_Piece_03 {type=number}
 Crown_Piece_04 {type=number}
 Crown_Piece_05 {type=number}
 Crown_Piece_06 {type=number}
 Neck_Plates_01 {type=number}
 Neck_Plates_02 {type=number}
]
